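# Physics asset for the FrozenReality back bling mesh, giving its dyn_ bones collision bodies and joints for secondary motion of paws and wings.
# A multi-capsule pack body anchors mirrored left and right chains: four-bone paws and three-bone wings running base, mounting, wing.
# Small capsules cover most links, while each wing adds a box to its capsule.
# Tight swing and twist limits restrict movement, with twist locked at the second paw joints.
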PhysicsAsset /BRCosmetics/Accessories/FORT_Backpacks/Backpack_M_MED_FrozenReality/Meshes/M_MED_FrozenReality_Pack_Physics
bodies {
 dyn_BackPack_FrozenReality [
  {
   shapeType=Capsule
   length=14.2722
   radius=6.0637
   shapeName=dyn_BackPack_FrozenReality_capsule
   center=<-0.8337 -1.2502 -0.0005>
   rotation=<rot 0 -100.8076 -89.9333>
  }
  {shapeType=Capsule, length=12.2044, radius=3.8119, shapeName=dyn_BackPack_FrozenReality_capsule, center=<-4.0883 -0.665 0>}
  {shapeType=Capsule, length=10.9426, radius=5.2134, shapeName=dyn_BackPack_FrozenReality_capsule, center=<0.536 0 0>}
  {shapeType=Capsule, length=5.9349, radius=5.2134, shapeName=dyn_BackPack_FrozenReality_capsule, center=<7.012 -1.7803 0>}
  {
   shapeType=Capsule
   length=5.9349
   radius=5.2134
   shapeName=dyn_BackPack_FrozenReality_capsule
   center=<-4.0279 8.9006 16.3012>
   rotation=<rot 0 -0.0001 30.7262>
  }
  {
   shapeType=Capsule
   length=5.9349
   radius=5.2134
   shapeName=dyn_BackPack_FrozenReality_capsule
   center=<-3.6289 2.9379 -17.8241>
   rotation=<rot 0 -0.0001 -14.2539>
  }
 ]
 dyn_Paw_1_L [{
   shapeType=Capsule
   length=2.9959
   radius=0.8745
   shapeName=dyn_Paw_1_L_capsule
   center=<0.6724 2.0775 -0.0165>
   rotation=<rot 0 -18.0436 -89.9864>
  }]
 dyn_Paw_1_R [{
   shapeType=Capsule
   length=2.9959
   radius=0.8745
   shapeName=dyn_Paw_1_R_capsule
   center=<2.0699 -0.0114 -0.0166>
   rotation=<rot 0 -94.3886 -89.9864>
  }]
 dyn_Paw_2_L [{shapeType=Capsule, length=0.505, radius=0.505, shapeName=dyn_Paw_2_L_capsule, center=<1.7242 0.4321 0.0097>, rotation=<rot 0 90.0425 -89.23>}]
 dyn_Paw_2_R [{
   shapeType=Capsule
   length=0.505
   radius=0.505
   shapeName=dyn_Paw_2_R_capsule
   center=<1.7489 -0.0674 0.0097>
   rotation=<rot 0 95.6358 -89.23>
  }]
 dyn_Paw_3_L [{
   shapeType=Capsule
   length=5.6983
   radius=1.8296
   shapeName=dyn_Paw_3_L_capsule
   center=<4.5793 1.2545 0.005>
   rotation=<rot 0 -86.8022 -89.8673>
  }]
 dyn_Paw_3_R [{
   shapeType=Capsule
   length=5.6983
   radius=1.8296
   shapeName=dyn_Paw_3_R_capsule
   center=<4.6714 -0.6032 0.005>
   rotation=<rot 0 -87.5194 -89.8673>
  }]
 dyn_Paw_Mounting_L [{
   shapeType=Capsule
   length=4.7923
   radius=0.8787
   shapeName=dyn_Paw_Mounting_L_capsule
   center=<4.0304 -0.1946 -0.0567>
   rotation=<rot 0 -90.4751 -89.1976>
  }]
 dyn_Paw_Mounting_R [{
   shapeType=Capsule
   length=4.7923
   radius=0.8787
   shapeName=dyn_Paw_Mounting_R_capsule
   center=<4.0382 0.3485 -0.0567>
   rotation=<rot 0 -91.0772 -89.1976>
  }]
 dyn_Wing_L [
  {shapeType=Capsule, length=20.7916, radius=3, shapeName=dyn_Wing_L_capsule, center=<8.0816 1.1173 0.5948>, rotation=<rot 0 88.2552 -81.1088>}
  {
   shapeType=Box
   extentX=27.9973
   extentY=2.5613
   extentZ=16.5135
   shapeName=dyn_Wing_L_box
   center=<14.3001 0.335 5.8013>
   rotation=<rot 8.9937 -4.1112 -0.6438>
  }
 ]
 dyn_Wing_Mounting_L [{shapeType=Capsule, length=0.505, radius=0.505, shapeName=dyn_Wing_Mounting_L_capsule, center=<1.4926 0 0>, rotation=<rot 0 -90 0>}]
 dyn_Wing_Mounting_R [{shapeType=Capsule, length=0.505, radius=0.505, shapeName=dyn_Wing_Mounting_R_capsule, center=<1.5254 0 0>, rotation=<rot 0 -90 0>}]
 dyn_Wing_R [
  {shapeType=Capsule, length=20.7916, radius=3, shapeName=dyn_Wing_R_capsule, center=<7.5654 -1.3937 2.0827>, rotation=<rot 0 91.7448 -65.7402>}
  {
   shapeType=Box
   extentX=29.071
   extentY=3.5439
   extentZ=17.4545
   shapeName=dyn_Wing_R_box
   center=<11.0641 -0.6753 5.0248>
   rotation=<rot 6.8751 2.2923 0.496>
  }
 ]
 dyn_wing_bace_L [{shapeType=Capsule, length=0.505, radius=0.505, shapeName=dyn_wing_bace_L_capsule, rotation=<rot 0 -90 0>}]
 dyn_wing_bace_R [{shapeType=Capsule, length=0.505, radius=0.505, shapeName=dyn_wing_bace_R_capsule, center=<0 -1.5188 0>, rotation=<rot 0 -90 0>}]
}
constraints [
 {bone1Name=dyn_Paw_1_L, bone2Name=dyn_Paw_Mounting_L, swing1LimitDegrees=5, swing2LimitDegrees=0, twistLimitDegrees=5}
 {bone1Name=dyn_Paw_1_R, bone2Name=dyn_Paw_Mounting_R, swing1LimitDegrees=4, swing2LimitDegrees=0, twistLimitDegrees=0}
 {bone1Name=dyn_Paw_2_L, bone2Name=dyn_Paw_1_L, swing1LimitDegrees=2, swing2LimitDegrees=2, twistLimitDegrees=45, twistMotion=Locked}
 {bone1Name=dyn_Paw_3_L, bone2Name=dyn_Paw_2_L, swing1LimitDegrees=7, swing2LimitDegrees=7, twistLimitDegrees=2}
 {bone1Name=dyn_Paw_2_R, bone2Name=dyn_Paw_1_R, swing1LimitDegrees=2, swing2LimitDegrees=2, twistLimitDegrees=45, twistMotion=Locked}
 {bone1Name=dyn_Paw_3_R, bone2Name=dyn_Paw_2_R, swing1LimitDegrees=7, swing2LimitDegrees=7, twistLimitDegrees=2}
 {bone1Name=dyn_Paw_Mounting_R, bone2Name=dyn_BackPack_FrozenReality, swing1LimitDegrees=16, swing2LimitDegrees=3, twistLimitDegrees=1}
 {bone1Name=dyn_Paw_Mounting_L, bone2Name=dyn_BackPack_FrozenReality, swing1LimitDegrees=16, swing2LimitDegrees=3, twistLimitDegrees=1}
 {bone1Name=dyn_Wing_R, bone2Name=dyn_Wing_Mounting_R, swing1LimitDegrees=16, swing2LimitDegrees=7, twistLimitDegrees=1}
 {bone1Name=dyn_Wing_Mounting_R, bone2Name=dyn_wing_bace_R, swing1LimitDegrees=17, swing2LimitDegrees=13, twistLimitDegrees=2}
 {bone1Name=dyn_Wing_L, bone2Name=dyn_Wing_Mounting_L, swing1LimitDegrees=16, swing2LimitDegrees=7, twistLimitDegrees=1}
 {bone1Name=dyn_Wing_Mounting_L, bone2Name=dyn_wing_bace_L, swing1LimitDegrees=17, swing2LimitDegrees=13, twistLimitDegrees=2}
 {bone1Name=dyn_wing_bace_R, bone2Name=dyn_BackPack_FrozenReality, swing1LimitDegrees=18, swing2LimitDegrees=11, twistLimitDegrees=1}
 {bone1Name=dyn_wing_bace_L, bone2Name=dyn_BackPack_FrozenReality, swing1LimitDegrees=18, swing2LimitDegrees=11, twistLimitDegrees=1}
]
