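# Animation Blueprint for the Frozen Reality backpack, animating its dangling physics, wings and wheels from player movement.
# The cached input pose feeds gameplay and front-end RigidBody simulations, blended by bIsInFrontEnd, then a chain of bone modifiers.
# Skydive float and dive states angle the wings.
# The update event flips simulation gravity upward, spins wheels from real time and toggles wheel-base poses during skydiving, jumping, falling or tactical sprinting.
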AnimBlueprint /BRCosmetics/Accessories/FORT_Backpacks/Backpack_M_MED_FrozenReality/Meshes/M_MED_FrozenReality_Pack_AnimBP
target_skeleton=&/M_MED_FrozenReality_Pack_Skeleton
graphs [
 AnimGraph {nodes [
   n1 = Get scail() -> (scail:Vector)
   n2 = RigidBody!(
    ComponentPose=Local To Component(LocalPose=Use cached pose 'PoseCache'())
    Alpha
    OverrideWorldGravity_X:double=0
    OverrideWorldGravity_Y:double=0
    OverrideWorldGravity_Z:double="Get Fly "() -> ("Fly ":double)
    ExternalForce
    ComponentLinearAccScale:Vector=@n1
    ComponentLinearVelScale:Vector=@n1
   ) -> (Pose) {
    bOverrideWorldGravity=true
    cachedBoundsScale=1
    evaluationResetTime=1
    overridePhysicsAsset=&/M_MED_FrozenReality_Pack_Physics
    alphaBoolBlend {blendOption=Linear}
    baseBoneRef {boneName=dyn_BackPack_FrozenReality}
    componentAppliedLinearAccClamp=<1000 1000 1000>
    componentLinearAccScale {z=0.1}
    componentLinearVelScale {z=0.1}
    overrideWorldGravity {z=100}
   }
   n3 = Component To Local(ComponentPose=@n2)
   n4 = Get scail()
   n5 = RigidBody(
    ComponentPose=Local To Component(LocalPose=Use cached pose 'PoseCache'())
    OverrideWorldGravity_X=0
    OverrideWorldGravity_Y=0
    OverrideWorldGravity_Z="Get Fly "()
    ComponentLinearAccScale=@n4
    ComponentLinearVelScale=@n4
   ) {
    bOverrideWorldGravity=true
    cachedBoundsScale=1
    evaluationResetTime=1
    overridePhysicsAsset=&/M_MED_FrozenReality_Pack_Physics_FrontEnd
    alphaBoolBlend {blendOption=Linear}
    baseBoneRef {boneName=dyn_BackPack_FrozenReality}
    componentAppliedLinearAccClamp=<1000 1000 1000>
    componentLinearAccScale {z=0.1}
    componentLinearVelScale {z=0.1}
    overrideWorldGravity {z=100}
   }
   n6 = Component To Local(ComponentPose=@n5)
   n7 = Get bIsInFrontEnd()
   n8 = "Transform (Modify) Bone - Bone: dyn_BackPack_FrozenReality"(
    ComponentPose=Local To Component(LocalPose=Blend Poses by bool(BlendPose_0=@n6, BlendPose_1=@n3, bActiveValue=@n7) {blendType=Linear})
    Alpha=1
    Translation="3,6,0"
    Rotation="0,-14,0"
    Scale="1,1,1"
   ) {
    rotationMode=BMM_Additive
    rotationSpace=BCS_BoneSpace
    translationMode=BMM_Replace
    translationSpace=BCS_BoneSpace
    alphaBoolBlend {blendOption=Linear}
    boneToModify {boneName=dyn_BackPack_FrozenReality}
    rotation {yaw=-14}
    translation=<3 6>
   }
   n9 = Get bIsSkydiveFloating()
   n10 = "Transform (Modify) Bone - Bone: dyn_wing_bace_R"!(ComponentPose=@n8, bAlphaBoolEnabled:bool=@n9, Rotation="-35,0,0") -> (Pose) {
    alphaInputType=Bool
    rotationMode=BMM_Additive
    rotationSpace=BCS_BoneSpace
    translationMode=BMM_Additive
    translationSpace=BCS_BoneSpace
    alphaBoolBlend {blendInTime=0.5, blendOption=Linear, blendOutTime=0.5}
    boneToModify {boneName=dyn_wing_bace_R}
    rotation {pitch=-35}
    translation=<-1 -1 0.5>
   }
   n11 = Get bIsSkydiveDiveMode()
   n12 = "Transform (Modify) Bone - Bone: dyn_Wing_Mounting_R"!(ComponentPose=@n10, bAlphaBoolEnabled:bool=@n11, Rotation="-15,20,0") -> (Pose) {
    alphaInputType=Bool
    rotationMode=BMM_Additive
    rotationSpace=BCS_BoneSpace
    translationMode=BMM_Additive
    translationSpace=BCS_BoneSpace
    alphaBoolBlend {blendInTime=0.5, blendOption=Linear, blendOutTime=0.5}
    boneToModify {boneName=dyn_Wing_Mounting_R}
    rotation {pitch=-15, yaw=20}
    translation=<-1 -1 0.5>
   }
   n13 = "Transform (Modify) Bone - Bone: dyn_wing_bace_L"!(ComponentPose=@n12, bAlphaBoolEnabled:bool=@n9, Rotation="-35,0,0") -> (Pose) {
    alphaInputType=Bool
    rotationMode=BMM_Additive
    rotationSpace=BCS_BoneSpace
    translationMode=BMM_Additive
    translationSpace=BCS_BoneSpace
    alphaBoolBlend {blendInTime=0.5, blendOption=Linear, blendOutTime=0.5}
    boneToModify {boneName=dyn_wing_bace_L}
    rotation {pitch=-35}
    translation=<-1 1 0.5>
   }
   n14 = "Transform (Modify) Bone - Bone: dyn_Wing_Mounting_L"!(ComponentPose=@n13, bAlphaBoolEnabled:bool=@n11, Rotation="-15,-20,0") -> (Pose) {
    alphaInputType=Bool
    rotationMode=BMM_Additive
    rotationSpace=BCS_BoneSpace
    translationMode=BMM_Additive
    translationSpace=BCS_BoneSpace
    alphaBoolBlend {blendInTime=0.5, blendOption=Linear, blendOutTime=0.5}
    boneToModify {boneName=dyn_Wing_Mounting_L}
    rotation {pitch=-15, yaw=-20}
    translation=<-1 1 0.5>
   }
   n15 = "Transform (Modify) Bone - Bone: dyn_Wheel_R"!(ComponentPose=@n14, Alpha=1, Rotation_Roll:float=Get alpha() -> (alpha:double), Rotation_Pitch:float=0, Rotation_Yaw:float=0) -> (Pose) {
    rotationMode=BMM_Replace
    rotationSpace=BCS_BoneSpace
    alphaBoolBlend {blendInTime=22, blendOption=Linear, blendOutTime=22}
    boneToModify {boneName=dyn_Wheel_R}
   }
   n16 = "Transform (Modify) Bone - Bone: dyn_Wheel_L"!(
    ComponentPose=@n15
    Alpha=1
    Rotation_Roll:float=Get Alpha Parachute() -> (Alpha Parachute:double)
    Rotation_Pitch:float=0
    Rotation_Yaw:float=0
   ) -> (Pose) {
    rotationMode=BMM_Replace
    rotationSpace=BCS_BoneSpace
    alphaBoolBlend {blendInTime=22, blendOption=Linear, blendOutTime=22}
    boneToModify {boneName=dyn_Wheel_L}
    rotation=<rot 0 0 11>
   }
   n17 = Get paden() -> (paden:bool)
   n18 = "Transform (Modify) Bone - Bone: dyn_weel_bace_R"!(ComponentPose=@n16, bAlphaBoolEnabled:bool=@n17, Translation="0,0,0", Rotation="29.662,5.766,-78.471", Scale="1,1,1") -> (Pose) {
    alphaInputType=Bool
    rotationMode=BMM_Replace
    rotationSpace=BCS_BoneSpace
    translationMode=BMM_Replace
    translationSpace=BCS_BoneSpace
    alphaBoolBlend {blendInTime=0.5, blendOption=Linear, blendOutTime=0.5}
    boneToModify {boneName=dyn_weel_bace_R}
    rotation=<rot 29.662 5.766 -78.471>
   }
   n19 = "Transform (Modify) Bone - Bone: dyn_weel_bace_L"!(ComponentPose=@n18, bAlphaBoolEnabled:bool=@n17, Translation="0,0,0", Rotation="-29.875,-5.813,-78.448", Scale="1,1,1") -> (Pose) {
    alphaInputType=Bool
    rotationMode=BMM_Replace
    rotationSpace=BCS_BoneSpace
    translationMode=BMM_Replace
    translationSpace=BCS_BoneSpace
    alphaBoolBlend {blendInTime=0.5, blendOption=Linear, blendOutTime=0.5}
    boneToModify {boneName=dyn_weel_bace_L}
    rotation=<rot -29.875 -5.813 -78.448>
   }
   n20 = Get off_on_wheels() -> (off_on_wheels:bool)
   n21 = "Transform (Modify) Bone - Bone: dyn_weel_bace_R"(ComponentPose=@n19, bAlphaBoolEnabled=@n20, Translation="0,0,0", Rotation="0,0,0", Scale="1,1,1") {
    alphaInputType=Bool
    rotationMode=BMM_Replace
    rotationSpace=BCS_BoneSpace
    alphaBoolBlend {blendInTime=0.5, blendOption=Linear, blendOutTime=0.5}
    boneToModify {boneName=dyn_weel_bace_R}
   }
   n22 = "Transform (Modify) Bone - Bone: dyn_weel_bace_L"(ComponentPose=@n21, bAlphaBoolEnabled=@n20, Translation="0,0,0", Rotation="0,0,0", Scale="1,1,1") {
    alphaInputType=Bool
    rotationMode=BMM_Replace
    rotationSpace=BCS_BoneSpace
    alphaBoolBlend {blendInTime=0.5, blendOption=Linear, blendOutTime=0.5}
    boneToModify {boneName=dyn_weel_bace_L}
   }
   n23 = "Transform (Modify) Bone - Bone: root"(ComponentPose=@n22, Translation="0,-3,-5", Rotation="0,0,0") {
    translationMode=BMM_Additive
    translationSpace=BCS_BoneSpace
    alphaBoolBlend {blendOption=Linear}
    boneToModify {boneName=root}
    translation {y=-3, z=-5}
   }
   Output Pose(Result=Component To Local(ComponentPose=@n23)) {layerGroup=None}
   Save cached pose 'PoseCache'(Pose=Input Pose()) {cacheName=PoseCache}
  ]}
 EventGraph {nodes [
   Event BlueprintUpdateAnimation()
   n1 = Get bIsTacticalSprinting()
   > n2 = Branch(Condition=OR Boolean!(A=Get bIsSkyDiving(), B=Get bIsSkydiveFloating(), C:bool=Get bIsJumping(), D:bool=@n1, E:bool=Get bIsFalling(), F:bool) -> (ReturnValue))
   > "Set Fly "("Fly ":double=400) -> (Output_Get:double)
   n3 = GetRealTimeSeconds()
   n4 = float * float(A=@n3, B=8) {is=title K2Node_PromotableOperator}
   > Set Alpha Parachute(Alpha Parachute:double=float * float(A=float + float(A=@n4, B=22), B=22)) -> (Output_Get:double)
   n5 = float * float(A=@n3, B=8)
   > Set alpha(alpha:double=float * float(A=float + float(A=@n5, B=1), B=22)) -> (Output_Get:double)
   > Set paden(paden:bool=true) -> (Output_Get:bool)
   > Set off_on_wheels(off_on_wheels:bool=true) -> (Output_Get:bool)
   > n6 = Branch(Condition=@n1)
   Set scail(execute=[@n6.then, @n7.then], scail:Vector="0,0,0.5") -> (Output_Get:Vector)
   Set scail(execute=@n6.else, scail="0,0,1")
   "Set Fly "(execute=@n2.else, "Fly "=-400)
   > Set Alpha Parachute(Alpha Parachute=0)
   > Set alpha(alpha=0)
   > Set paden(paden=false)
   > n7 = Set off_on_wheels(off_on_wheels=false)
   TryGetPawnOwner()
  ]}
]
variables [
 alpha {type=number}
 Alpha Parachute {type=number}
 "Fly " {type=number}
 off_on_wheels {type=boolean}
 paden {type=boolean}
 scail {type=Vector}
]
